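# Material function that applies a stack of four-layer surface effects to an incoming material attributes set, named as the tangent-space variant.
# It chains color replace, multiply tint, velvet and emissive flicker sub-functions in that order.
# A vector4 layer mask feeds every stage, and a scalar texture mask feeds tint and velvet.
# Both masks travel through named reroutes and are passed out unchanged beside the resulting material.
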MaterialFunction /BRCoreMaterials/Materials/MaterialFunctions/MF_4Layer_Effects_TangentSpace
graph {nodes [
  n1 = MaterialFunctionCall!(Material=FunctionInput() {inputName=Material, inputType=FunctionInput_MaterialAttributes}, Layer Mask=NamedRerouteUsage()) {materialFunction=&/MF_4Layer_ColorReplace}
  n2 = MaterialFunctionCall!(
   Material=@n1.Material
   Color Replace Mask=@n1.Material
   Color Replace Toggle=@n1.Material
   Layer Mask=NamedRerouteUsage()
   Layer Texture Mask=NamedRerouteUsage()
  ) {materialFunction=&/MF_4Layer_Tint_Multiply}
  n3 = NamedRerouteUsage()
  n4 = NamedRerouteUsage()
  FunctionOutput(None=MaterialFunctionCall!(
    Material=MaterialFunctionCall!(Material=@n2.Material, Layer Mask=@n3, Layer Texture Mask=@n4) {materialFunction=&/MF_4Layer_Velvet}.Result
    Layer Mask=NamedRerouteUsage()
   ) {materialFunction=&/MF_4Layer_EmissiveFlicker}.Result) {outputName=Material}
  FunctionOutput(None=NamedRerouteDeclaration(None=FunctionInput() {bUsePreviewValueAsDefault=true, inputName=Layer Mask, inputType=FunctionInput_Vector4, sortPriority=1}) {name=Layer Mask, nodeColor=<rgba 1 0.0077 0.1226 1>}) {outputName=Layer Mask, sortPriority=1}
  FunctionOutput(None=NamedRerouteDeclaration(None=FunctionInput() {bUsePreviewValueAsDefault=true, inputName=Layer Texture Mask, inputType=FunctionInput_Scalar, sortPriority=2}) {name=Layer Texture Mask, nodeColor=<rgba 0 0.9255 1 1>}) {outputName=Layer Texture Mask, sortPriority=2}
 ]}
