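# Material function that applies the world-space effects stack to a four-layer material's attributes.
# Chains the 4Layer sub-functions in order: DisplacementMagnitude, Cover, Variation, DFX, RVT, DestructibleDamage, then WPORotation.
# The Layer Mask and Layer Texture Mask inputs feed the stages that use them through named reroutes.
# It outputs the modified material and passes both masks through unchanged.
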MaterialFunction /BRCoreMaterials/Materials/MaterialFunctions/MF_4Layer_Effects_WorldSpace
graph {nodes [
  n1 = MaterialFunctionCall!(Material=FunctionInput() {inputName=Material, inputType=FunctionInput_MaterialAttributes}, Layer Mask=NamedRerouteUsage()) {materialFunction=&/MF_4Layer_DisplacementMagnitude}
  n2 = MaterialFunctionCall!(Material=@n1.Result, Displacement Texture=@n1.Result, Layer Mask=NamedRerouteUsage(), Layer Texture Mask=NamedRerouteUsage()) {materialFunction=&/MF_4Layer_Cover}
  n3 = NamedRerouteUsage()
  n4 = NamedRerouteUsage()
  n5 = MaterialFunctionCall!(
   Material=MaterialFunctionCall!(In=MaterialFunctionCall!(Material=@n2.Result, Layer Mask=@n3, Layer Texture Mask=@n4) {materialFunction=&/MF_4Layer_Variation}.Result) {materialFunction=&/MF_4Layer_DFX}.Result
   Layer Mask=NamedRerouteUsage()
   Layer Texture Mask=NamedRerouteUsage()
  ) {materialFunction=&/MF_4Layer_RVT}
  FunctionOutput(None=MaterialFunctionCall!(
    Material=MaterialFunctionCall!(
     Material_In=@n5.Result
     UseAlternativeCustomPrimDataForDestruction
     UseDestructibleDamage
     UseDamageWobbleWPO
     WPO Distance Cutoff
     WorldPosition
     UseAlternateHitGradientOutput
     DestructibleDamageAmount
     WSDeformationCenter
     WSDamageRadius
    ) {materialFunction=&/MF_4Layer_DestructibleDamage}.Material
    Layer Mask=NamedRerouteUsage()
   ) {materialFunction=&/MF_4Layer_WPORotation}.Result) {outputName=Material}
  FunctionOutput(None=NamedRerouteDeclaration(None=FunctionInput() {inputName=Layer Mask, inputType=FunctionInput_Vector4, sortPriority=1}) {name=Layer Mask, nodeColor=<rgba 1 0.0405 0.2348 1>}) {outputName=Layer Mask, sortPriority=1}
  FunctionOutput(None=NamedRerouteDeclaration(None=FunctionInput() {inputName=Layer Texture Mask, inputType=FunctionInput_Scalar, sortPriority=2}) {name=Layer Texture Mask, nodeColor=<rgba 0 1 0.6781 1>}) {outputName=Layer Texture Mask, sortPriority=2}
 ]}
