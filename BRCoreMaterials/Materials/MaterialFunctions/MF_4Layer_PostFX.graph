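# Material function that applies the post effects stage for four-layer materials to a set of material attributes.
# It passes the incoming attributes, a four-channel layer mask and a scalar layer texture mask into MF_4Layer_Sheen, with that function's TangentNormals switch fixed to false.
# The sheen result then runs through MF_4Layer_ValuePostProcess, which receives only the attributes, and returns as the Material output.
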MaterialFunction /BRCoreMaterials/Materials/MaterialFunctions/MF_4Layer_PostFX
graph {nodes [
  n1 = FunctionInput() {inputName=Material, inputType=FunctionInput_MaterialAttributes}
  n2 = NamedRerouteUsage()
  n3 = NamedRerouteUsage()
  n4 = StaticBool()
  FunctionOutput(None=MaterialFunctionCall!(Material=MaterialFunctionCall!(Material=@n1, Layer Mask=@n2, Layer Texture Mask=@n3, TangentNormals=@n4) {materialFunction=&/MF_4Layer_Sheen}.Result) {materialFunction=&/MF_4Layer_ValuePostProcess}.Result) {outputName=Material}
  NamedRerouteDeclaration(None=FunctionInput() {inputName=Layer Mask, inputType=FunctionInput_Vector4, sortPriority=1}) {name=Layer Mask, nodeColor=<rgba 1 0.0405 0.2348 1>}
  NamedRerouteDeclaration(None=FunctionInput() {inputName=Layer Texture Mask, inputType=FunctionInput_Scalar, sortPriority=2}) {name=Layer Texture Mask, nodeColor=<rgba 0 1 0.6781 1>}
 ]}
