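# Material function that builds a switchable one-to-four-layer surface material before effects, outputting material attributes, layer mask and layer texture mask.
# Integer U tiles of UV1 select the layer, and their fractional edges form a displacement zipper mask.
# Each layer runs MF_4Layer_Layer with its own textures and parameters, then gets blended.
# Optional switches add masked per-layer world-space UVs with normal correction, and scroll or flipbook animation with temporal responsiveness.
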MaterialFunction /BRCoreMaterials/Materials/MaterialFunctions/MF_4Layer_PreFX
graph {nodes [
  n1 = NamedRerouteUsage()
  n2 = BlendMaterialAttributes(A=@n1, B=NamedRerouteUsage(), Alpha=ComponentMask(None=NamedRerouteUsage()) {g=true})
  n3 = BlendMaterialAttributes(A=@n2, B=NamedRerouteUsage(), Alpha=ComponentMask(None=NamedRerouteUsage()) {b=true})
  n4 = BlendMaterialAttributes(A=@n3, B=NamedRerouteUsage(), Alpha=ComponentMask(None=NamedRerouteUsage()) {a=true})
  n5 = StaticSwitchParameter(
   True=StaticSwitchParameter(True=StaticSwitchParameter(True=@n4, False=@n3) {group=- Layer Switches, parameterName=Use 4 Layers, sortPriority=2}, False=@n2) {group=- Layer Switches, parameterName=Use 3 Layers, sortPriority=1}
   False=@n1
  ) {group=- Layer Switches, parameterName=Use 2 Layers, sortPriority=0}
  n6 = GetMaterialAttributes(None=@n5) {attributeGetTypes [#1]}
  n7 = MaterialFunctionCall!(TangentNormals=@n6.Normal, Mask=NamedRerouteUsage()) {materialFunction=&/MF_4Layer_WorldSpaceUVs_TangentToWorld}
  n8 = NamedRerouteUsage()
  n9 = SetMaterialAttributes!(
   MaterialAttributes=@n6.MaterialAttributes
   Normal=LinearInterpolate(
    A=@n6.Normal
    B=Transform(None=@n7.World Normals) {transformSourceType=TRANSFORMSOURCE_World, transformType=TRANSFORM_Tangent}
    Alpha=Saturate(None=@n8)
   )
  ) {attributeSetTypes [#1]}
  n10 = StaticBoolParameter() {group=- Effect Switches, parameterName=World Space UVs}
  FunctionOutput(None=SetMaterialAttributes!(MaterialAttributes=StaticSwitch(True=@n9, False=@n5, Value=@n10), TemporalResponsiveness=NamedRerouteUsage()) {attributeSetTypes [#2]}) {outputName=Material}
  FunctionOutput(None=NamedRerouteUsage()) {outputName=Layer Mask, sortPriority=1}
  FunctionOutput(None=MaterialFunctionCall!(RGBA Mask=NamedRerouteUsage(), L1=NamedRerouteUsage(), L2=NamedRerouteUsage(), L3=NamedRerouteUsage(), L4=NamedRerouteUsage()) {materialFunction=&/MF_4Layer_ScalarMasking}.Result) {outputName=Layer Texture Mask, sortPriority=2}
  n11 = TextureObjectParameter() {
   group=1 Layer
   parameterName=Diffuse
   sortPriority=-20
   texture=/BRCoreMaterials/Materials/Textures/T_BaseColor
   channelNames {a=Mask, b="", g="", r=""}
  }
  n12 = TextureObjectParameter() {group=1 Layer, parameterName=Normals, samplerType=SAMPLERTYPE_Normal, sortPriority=-19, texture=/BRCoreMaterials/Materials/Textures/T_Normal}
  n13 = TextureObjectParameter() {
   group=1 Layer
   parameterName=SpecularMasks
   samplerType=SAMPLERTYPE_Masks
   sortPriority=-18
   texture=/BRCoreMaterials/Materials/Textures/T_Specular
   channelNames {a="", b=Roughness, g=Metal, r=Specular}
  }
  n14 = TextureObjectParameter() {
   group=1 Layer
   parameterName=DisplacementMask
   samplerType=SAMPLERTYPE_LinearGrayscale
   sortPriority=-17
   texture=/BRCoreMaterials/Materials/Textures/T_Displacement
  }
  n15 = MaterialFunctionCall!(
   Diffuse=@n11
   Normals=@n12
   SpecularMasks=@n13
   WorldSpaceUVs Mode=NamedRerouteUsage()
   Opacity Mask=ScalarParameter() {group=1 Layer, parameterName=L1 Opacity Mask, sliderMax=1, sortPriority=3}
   Emissive Color=VectorParameter() {group=1 Layer, parameterName=L1 Emissive Color, sortPriority=4, defaultValue=<rgba 0 0 0 1>}.RGB
   Displacement=@n14
   Displacement Zipper Mask=NamedRerouteUsage()
   Displacement Zipper Exp=ScalarParameter() {defaultValue=100, group=1 Layer, parameterName=L1 Displacement Zipper, sortPriority=9}
   Emissive Lumen Boost=ScalarParameter() {defaultValue=1, group=1 Layer, parameterName=L1 Emissive Lumen Boost, sortPriority=7}
   Displacement WorldSpace UVY=NamedRerouteUsage()
   WorldSpaceUVs=NamedRerouteUsage()
   Displacement WorldSpace Mask=NamedRerouteUsage()
   Displacement WorldSpace UVZ=NamedRerouteUsage()
   Displacement WorldSpace UVX=NamedRerouteUsage()
   Emissive Intensity=ScalarParameter() {defaultValue=1, group=1 Layer, parameterName=L1 Emissive Intensity, sortPriority=6}
   Emissive FromBaseColor=ScalarParameter() {group=1 Layer, parameterName=L1 Emissive From BaseColor, sliderMax=1, sortPriority=5}
   Tiling=ScalarParameter() {defaultValue=1, group=1 Layer, parameterName=L1 Tiling, sortPriority=-1}
   UV Offset=AppendVector(
    A=ScalarParameter() {group=1 Layer | UV Offsets, parameterName=L1 UV Offset X}
    B=ScalarParameter() {group=1 Layer | UV Offsets, parameterName=L1 UV Offset Y}
   )
   Texcoords=NamedRerouteUsage()
  ) {materialFunction=&/MF_4Layer_Layer}
  NamedRerouteDeclaration(None=@n15.Material) {name=Layer1 Material, nodeColor=<rgba 1 0 0.5859 1>}
  n16 = TextureObjectParameter() {
   group=2 Layer
   parameterName=Diffuse_Texture_2
   sortPriority=-20
   texture=&T_BaseColor
   channelNames {a=Mask, b="", g="", r=""}
  }
  n17 = TextureObjectParameter() {
   group=2 Layer
   parameterName=Normals_Texture_2
   samplerType=SAMPLERTYPE_Normal
   sortPriority=-19
   texture=&T_Normal
  }
  n18 = TextureObjectParameter() {
   group=2 Layer
   parameterName=SpecularMasks_2
   samplerType=SAMPLERTYPE_Masks
   sortPriority=-18
   texture=&T_Specular
   channelNames {a="", b=Roughness, g=Metal, r=Specular}
  }
  n19 = TextureObjectParameter() {
   group=2 Layer
   parameterName=DisplacementMask_2
   samplerType=SAMPLERTYPE_LinearGrayscale
   sortPriority=-17
   texture=&T_Displacement
  }
  n20 = MaterialFunctionCall(
   Diffuse=@n16
   Normals=@n17
   SpecularMasks=@n18
   WorldSpaceUVs Mode=NamedRerouteUsage()
   Opacity Mask=ScalarParameter() {group=2 Layer, parameterName=L2 Opacity Mask, sliderMax=1, sortPriority=3}
   Emissive Color=VectorParameter() {group=2 Layer, parameterName=L2 Emissive Color, sortPriority=4, defaultValue=<rgba 0 0 0 1>}.RGB
   Displacement=@n19
   Displacement Zipper Mask=NamedRerouteUsage()
   Displacement Zipper Exp=ScalarParameter() {defaultValue=100, group=2 Layer, parameterName=L2 Displacement Zipper, sortPriority=9}
   Emissive Lumen Boost=ScalarParameter() {defaultValue=1, group=2 Layer, parameterName=L2 Emissive Lumen Boost, sortPriority=7}
   Displacement WorldSpace UVY=NamedRerouteUsage()
   WorldSpaceUVs=NamedRerouteUsage()
   Displacement WorldSpace Mask=NamedRerouteUsage()
   Displacement WorldSpace UVZ=NamedRerouteUsage()
   Displacement WorldSpace UVX=NamedRerouteUsage()
   Emissive Intensity=ScalarParameter() {defaultValue=1, group=2 Layer, parameterName=L2 Emissive Intensity, sortPriority=6}
   Emissive FromBaseColor=ScalarParameter() {group=2 Layer, parameterName=L2 Emissive From BaseColor, sliderMax=1, sortPriority=5}
   Tiling=ScalarParameter() {defaultValue=1, group=2 Layer, parameterName=L2 Tiling, sortPriority=-1}
   UV Offset=AppendVector(
    A=ScalarParameter() {group=2 Layer | UV Offsets, parameterName=L2 UV Offset X}
    B=ScalarParameter() {group=2 Layer | UV Offsets, parameterName=L2 UV Offset Y}
   )
   Texcoords=NamedRerouteUsage()
  ) {materialFunction=&MF_4Layer_Layer}
  NamedRerouteDeclaration(None=@n20.Material) {name=Layer2 Material, nodeColor=<rgba 0 0.9062 1 1>}
  n21 = ScalarParameter() {
   controlType=Enumeration
   enumeration=/BRCoreMaterials/Materials/Enums/EN_4Layer_GenericToggle
   group=Animated UVs
   parameterName=L1 AnimatedUVs Flipbook
   sortPriority=3
  }
  n22 = ScalarParameter() {
   controlType=Enumeration
   enumeration=&EN_4Layer_GenericToggle
   group=Animated UVs
   parameterName=L2 AnimatedUVs Flipbook
   sortPriority=10
  }
  n23 = ScalarParameter() {
   controlType=Enumeration
   enumeration=&EN_4Layer_GenericToggle
   group=Animated UVs
   parameterName=L3 AnimatedUVs Flipbook
   sortPriority=17
  }
  n24 = ScalarParameter() {
   controlType=Enumeration
   enumeration=&EN_4Layer_GenericToggle
   group=Animated UVs
   parameterName=L4 AnimatedUVs Flipbook
   sortPriority=24
  }
  n25 = MaterialFunctionCall(RGBA Mask=NamedRerouteUsage(), L1=@n21, L2=@n22, L3=@n23, L4=@n24) {materialFunction=&MF_4Layer_ScalarMasking}
  n26 = NamedRerouteUsage()
  n27 = NamedRerouteUsage()
  NamedRerouteDeclaration(None=StaticSwitchParameter(True=IfThenElse!(Condition=@n25.Result, True=@n26, False=@n27), False=TextureCoordinate()) {group=- Effect Switches, parameterName=Animated UVs}) {name=Texcoord UVs, nodeColor=<rgba 0.6172 0 1 1>}
  n28 = NamedRerouteUsage()
  n29 = ScalarParameter() {
   controlType=Enumeration
   enumeration=/BRCoreMaterials/Materials/Enums/EN_4Layer_WorldSpaceUVs_Toggle
   group=World Space UVs
   parameterName=L1 WorldSpaceUVs Mode
   sortPriority=0
  }
  n30 = ScalarParameter() {
   controlType=Enumeration
   enumeration=&EN_4Layer_WorldSpaceUVs_Toggle
   group=World Space UVs
   parameterName=L2 WorldSpaceUVs Mode
   sortPriority=3
  }
  n31 = ScalarParameter() {
   controlType=Enumeration
   enumeration=&EN_4Layer_WorldSpaceUVs_Toggle
   group=World Space UVs
   parameterName=L3 WorldSpaceUVs Mode
   sortPriority=6
  }
  n32 = ScalarParameter() {
   controlType=Enumeration
   enumeration=&EN_4Layer_WorldSpaceUVs_Toggle
   group=World Space UVs
   parameterName=L4 WorldSpaceUVs Mode
   sortPriority=9
  }
  n33 = ScalarParameter() {
   defaultValue=0.1
   enumeration=&EN_4Layer_WorldSpaceUVs_Toggle
   group=World Space UVs
   parameterName=L1 WorldSpaceUVs Blend
   sliderMax=0.5
   sortPriority=2
  }
  n34 = ScalarParameter() {
   defaultValue=0.1
   enumeration=&EN_4Layer_WorldSpaceUVs_Toggle
   group=World Space UVs
   parameterName=L2 WorldSpaceUVs Blend
   sliderMax=0.5
   sortPriority=5
  }
  n35 = ScalarParameter() {
   defaultValue=0.1
   enumeration=&EN_4Layer_WorldSpaceUVs_Toggle
   group=World Space UVs
   parameterName=L3 WorldSpaceUVs Blend
   sliderMax=0.5
   sortPriority=8
  }
  n36 = ScalarParameter() {
   defaultValue=0.1
   enumeration=&EN_4Layer_WorldSpaceUVs_Toggle
   group=World Space UVs
   parameterName=L4 WorldSpaceUVs Blend
   sliderMax=0.5
   sortPriority=11
  }
  n37 = MaterialFunctionCall!(
   Mode=NamedRerouteDeclaration(None=MaterialFunctionCall(RGBA Mask=@n28, L1=@n29, L2=@n30, L3=@n31, L4=@n32) {materialFunction=&MF_4Layer_ScalarMasking}.Result) {name=WorldSpace Mode, nodeColor=<rgba 0.9453 0 1 1>}
   Blend Radius=MaterialFunctionCall(RGBA Mask=NamedRerouteUsage(), L1=@n33, L2=@n34, L3=@n35, L4=@n36) {materialFunction=&MF_4Layer_ScalarMasking}.Result
  ) {materialFunction=&/MF_4Layer_WorldSpaceUVs_Masks}
  NamedRerouteDeclaration(None=@n37.Smooth) {name=WorldSpace Smooth Mask, nodeColor=<rgba 0.6641 0 1 1>}
  n38 = ScalarParameter() {
   defaultValue=384
   enumeration=&EN_4Layer_WorldSpaceUVs_Toggle
   group=World Space UVs
   parameterName=L1 WorldSpaceUVs Scale
   sortPriority=1
  }
  n39 = ScalarParameter() {
   defaultValue=384
   enumeration=&EN_4Layer_WorldSpaceUVs_Toggle
   group=World Space UVs
   parameterName=L2 WorldSpaceUVs Scale
   sortPriority=4
  }
  n40 = ScalarParameter() {
   defaultValue=384
   enumeration=&EN_4Layer_WorldSpaceUVs_Toggle
   group=World Space UVs
   parameterName=L3 WorldSpaceUVs Scale
   sortPriority=7
  }
  n41 = ScalarParameter() {
   defaultValue=384
   enumeration=&EN_4Layer_WorldSpaceUVs_Toggle
   group=World Space UVs
   parameterName=L4 WorldSpaceUVs Scale
   sortPriority=10
  }
  n42 = MaterialFunctionCall!(
   UV Scale=MaterialFunctionCall(RGBA Mask=NamedRerouteUsage(), L1=@n38, L2=@n39, L3=@n40, L4=@n41) {materialFunction=&MF_4Layer_ScalarMasking}.Result
   Mask=NamedRerouteDeclaration(None=@n37.Dithered) {name=Dithered Mask, nodeColor=<rgba 1 0.8499 0 1>}
  ) {materialFunction=&/MF_4Layer_WorldSpaceUVs_UVs}
  NamedRerouteDeclaration(None=AppendVector(A=@n42.UVs, B=@n42.UVs)) {name=WorldSpace UVs, nodeColor=<rgba 0.6641 0 1 1>}
  NamedRerouteDeclaration(None=@n42.UV Z) {name=WorldSpace UV Z, nodeColor=<rgba 0.0434 0 1 1>}
  NamedRerouteDeclaration(None=@n42.UV X) {name=WorldSpace UV X, nodeColor=<rgba 1 0 0.0123 1>}
  NamedRerouteDeclaration(None=@n42.UV Y) {name=WorldSpace UV Y, nodeColor=<rgba 0.0421 1 0 1>}
  n43 = Abs(None=NamedRerouteUsage())
  n44 = Abs(None=NamedRerouteUsage())
  n45 = Ceil(None=Max(A=Max(A=@n43, B=@n44), B=NamedRerouteUsage()))
  NamedRerouteDeclaration(None=StaticSwitchParameter(True=Saturate(None=@n45), False=Constant()) {group=- Effect Switches, parameterName=Animated UVs}) {name=Temporal Responsiveness, nodeColor=<rgba 1 0.0469 0 1>}
  n46 = TextureCoordinate()
  n47 = MaterialFunctionCall(
   RGBA Mask=NamedRerouteUsage()
   L1=ScalarParameter() {defaultValue=2, group=Animated UVs, parameterName=L1 AnimatedUVs Flipbook Columns, sortPriority=4}
   L2=ScalarParameter() {defaultValue=2, group=Animated UVs, parameterName=L2 AnimatedUVs Flipbook Columns, sortPriority=11}
   L3=ScalarParameter() {defaultValue=2, group=Animated UVs, parameterName=L3 AnimatedUVs Flipbook Columns, sortPriority=18}
   L4=ScalarParameter() {defaultValue=2, group=Animated UVs, parameterName=L4 AnimatedUVs Flipbook Columns, sortPriority=25}
  ) {materialFunction=&MF_4Layer_ScalarMasking}
  n48 = MaterialFunctionCall(
   RGBA Mask=NamedRerouteUsage()
   L1=ScalarParameter() {defaultValue=2, group=Animated UVs, parameterName=L1 AnimatedUVs Flipbook Rows, sortPriority=5}
   L2=ScalarParameter() {defaultValue=2, group=Animated UVs, parameterName=L2 AnimatedUVs Flipbook Rows, sortPriority=12}
   L3=ScalarParameter() {defaultValue=2, group=Animated UVs, parameterName=L3 AnimatedUVs Flipbook Rows, sortPriority=19}
   L4=ScalarParameter() {defaultValue=2, group=Animated UVs, parameterName=L4 AnimatedUVs Flipbook Rows, sortPriority=26}
  ) {materialFunction=&MF_4Layer_ScalarMasking}
  n49 = AppendVector(A=@n47.Result, B=@n48.Result)
  n50 = Multiply(A=@n47.Result, B=@n48.Result)
  n51 = MaterialFunctionCall(
   RGBA Mask=NamedRerouteUsage()
   L1=ScalarParameter() {group=Animated UVs, parameterName=L1 AnimatedUVs Flipbook Speed, sortPriority=6}
   L2=ScalarParameter() {group=Animated UVs, parameterName=L2 AnimatedUVs Flipbook Speed, sortPriority=13}
   L3=ScalarParameter() {group=Animated UVs, parameterName=L3 AnimatedUVs Flipbook Speed, sortPriority=20}
   L4=ScalarParameter() {group=Animated UVs, parameterName=L4 AnimatedUVs Flipbook Speed, sortPriority=27}
  ) {materialFunction=&MF_4Layer_ScalarMasking}
  n52 = Frac(None=Multiply(A=Time(), B=NamedRerouteDeclaration(None=@n51.Result) {name=Flipbook Speed, nodeColor=<rgba 0 1 0.2031 1>}) {constB=0.2}) {desc=0-1 animation}
  n53 = Floor(None=Multiply(A=AppendVector(A=@n50, B=@n48.Result), B=AppendVector(A=@n52, B=@n52)))
  NamedRerouteDeclaration(None=Add(A=Divide(A=@n46, B=@n49), B=Divide(A=@n53, B=@n49))) {name=Flipbook, nodeColor=<rgba 0 0.1562 1 1>}
  NamedRerouteDeclaration(None=@n20.Mask) {name=Layer2 Mask, nodeColor=<rgba 1 1 1 1>}
  n54 = MaterialFunctionCall(
   RGBA Mask=NamedRerouteUsage()
   L1=ScalarParameter() {group=Animated UVs, parameterName=L1 AnimatedUVs Scroll X, sortPriority=1}
   L2=ScalarParameter() {group=Animated UVs, parameterName=L2 AnimatedUVs Scroll X, sortPriority=8}
   L3=ScalarParameter() {group=Animated UVs, parameterName=L3 AnimatedUVs Scroll X, sortPriority=15}
   L4=ScalarParameter() {group=Animated UVs, parameterName=L4 AnimatedUVs Scroll X, sortPriority=22}
  ) {materialFunction=&MF_4Layer_ScalarMasking}
  n55 = NamedRerouteUsage()
  n56 = ScalarParameter() {group=Animated UVs, parameterName=L1 AnimatedUVs Scroll Y, sortPriority=2}
  n57 = ScalarParameter() {group=Animated UVs, parameterName=L2 AnimatedUVs Scroll Y, sortPriority=9}
  n58 = ScalarParameter() {group=Animated UVs, parameterName=L3 AnimatedUVs Scroll Y, sortPriority=16}
  n59 = ScalarParameter() {group=Animated UVs, parameterName=L4 AnimatedUVs Scroll Y, sortPriority=23}
  n60 = AppendVector(
   A=NamedRerouteDeclaration(None=@n54.Result) {name=Scroll Speed X, nodeColor=<rgba 1 0.8438 0 1>}
   B=NamedRerouteDeclaration(None=MaterialFunctionCall(RGBA Mask=@n55, L1=@n56, L2=@n57, L3=@n58, L4=@n59) {materialFunction=&MF_4Layer_ScalarMasking}.Result) {name=Scroll Speed Y, nodeColor=<rgba 0.5234 0 1 1>}
  )
  n61 = Time()
  NamedRerouteDeclaration(None=Add(A=Multiply(A=@n60, B=@n61), B=TextureCoordinate())) {name=Scroll UVs, nodeColor=<rgba 0.4766 1 0 1>}
  n62 = TextureObjectParameter() {
   group=3 Layer
   parameterName=Diffuse_Texture_3
   sortPriority=-20
   texture=&T_BaseColor
   channelNames {a=Mask, b="", g="", r=""}
  }
  n63 = TextureObjectParameter() {
   group=3 Layer
   parameterName=Normals_Texture_3
   samplerType=SAMPLERTYPE_Normal
   sortPriority=-19
   texture=&T_Normal
  }
  n64 = TextureObjectParameter() {
   group=3 Layer
   parameterName=SpecularMasks_3
   samplerType=SAMPLERTYPE_Masks
   sortPriority=-18
   texture=&T_Specular
   channelNames {a="", b=Roughness, g=Metal, r=Specular}
  }
  n65 = TextureObjectParameter() {
   group=3 Layer
   parameterName=DisplacementMask_3
   samplerType=SAMPLERTYPE_LinearGrayscale
   sortPriority=-17
   texture=&T_Displacement
  }
  n66 = MaterialFunctionCall(
   Diffuse=@n62
   Normals=@n63
   SpecularMasks=@n64
   WorldSpaceUVs Mode=NamedRerouteUsage()
   Opacity Mask=ScalarParameter() {group=3 Layer, parameterName=L3 Opacity Mask, sliderMax=2, sortPriority=3}
   Emissive Color=VectorParameter() {group=3 Layer, parameterName=L3 Emissive Color, sortPriority=4, defaultValue=<rgba 0 0 0 1>}.RGB
   Displacement=@n65
   Displacement Zipper Mask=NamedRerouteUsage()
   Displacement Zipper Exp=ScalarParameter() {defaultValue=100, group=3 Layer, parameterName=L3 Displacement Zipper, sortPriority=9}
   Emissive Lumen Boost=ScalarParameter() {defaultValue=1, group=3 Layer, parameterName=L3 Emissive Lumen Boost, sortPriority=7}
   Displacement WorldSpace UVY=NamedRerouteUsage()
   WorldSpaceUVs=NamedRerouteUsage()
   Displacement WorldSpace Mask=NamedRerouteUsage()
   Displacement WorldSpace UVZ=NamedRerouteUsage()
   Displacement WorldSpace UVX=NamedRerouteUsage()
   Emissive Intensity=ScalarParameter() {defaultValue=1, group=3 Layer, parameterName=L3 Emissive Intensity, sortPriority=6}
   Emissive FromBaseColor=ScalarParameter() {group=3 Layer, parameterName=L3 Emissive From BaseColor, sliderMax=1, sortPriority=5}
   Tiling=ScalarParameter() {defaultValue=1, group=3 Layer, parameterName=L3 Tiling, sortPriority=-1}
   UV Offset=AppendVector(
    A=ScalarParameter() {group=3 Layer | UV Offsets, parameterName=L3 UV Offset X}
    B=ScalarParameter() {group=3 Layer | UV Offsets, parameterName=L3 UV Offset Y}
   )
   Texcoords=NamedRerouteUsage()
  ) {materialFunction=&MF_4Layer_Layer}
  NamedRerouteDeclaration(None=@n66.Material) {name=Layer3 Material, nodeColor=<rgba 0 1 0.0604 1>}
  NamedRerouteDeclaration(None=@n66.Mask) {name=Layer3 Mask, nodeColor=<rgba 1 1 1 1>}
  n67 = TextureObjectParameter() {
   group=4 Layer
   parameterName=Diffuse_Texture_4
   sortPriority=-20
   texture=&T_BaseColor
   channelNames {a=Mask, b="", g="", r=""}
  }
  n68 = TextureObjectParameter() {
   group=4 Layer
   parameterName=Normals_Texture_4
   samplerType=SAMPLERTYPE_Normal
   sortPriority=-19
   texture=&T_Normal
  }
  n69 = TextureObjectParameter() {
   group=4 Layer
   parameterName=SpecularMasks_4
   samplerType=SAMPLERTYPE_Masks
   sortPriority=-18
   texture=&T_Specular
   channelNames {a="", b=Roughness, g=Metal, r=Specular}
  }
  n70 = TextureObjectParameter() {
   group=4 Layer
   parameterName=DisplacementMask_4
   samplerType=SAMPLERTYPE_LinearGrayscale
   sortPriority=-17
   texture=&T_Displacement
  }
  n71 = MaterialFunctionCall(
   Diffuse=@n67
   Normals=@n68
   SpecularMasks=@n69
   WorldSpaceUVs Mode=NamedRerouteUsage()
   Opacity Mask=ScalarParameter() {group=4 Layer, parameterName=L4 Opacity Mask, sliderMax=1, sortPriority=3}
   Emissive Color=VectorParameter() {group=4 Layer, parameterName=L4 Emissive Color, sortPriority=4, defaultValue=<rgba 0 0 0 1>}.RGB
   Displacement=@n70
   Displacement Zipper Mask=NamedRerouteUsage()
   Displacement Zipper Exp=ScalarParameter() {defaultValue=100, group=4 Layer, parameterName=L4 Displacement Zipper, sortPriority=9}
   Emissive Lumen Boost=ScalarParameter() {defaultValue=1, group=4 Layer, parameterName=L4 Emissive Lumen Boost, sortPriority=7}
   Displacement WorldSpace UVY=NamedRerouteUsage()
   WorldSpaceUVs=NamedRerouteUsage()
   Displacement WorldSpace Mask=NamedRerouteUsage()
   Displacement WorldSpace UVZ=NamedRerouteUsage()
   Displacement WorldSpace UVX=NamedRerouteUsage()
   Emissive Intensity=ScalarParameter() {defaultValue=1, group=4 Layer, parameterName=L4 Emissive Intensity, sortPriority=6}
   Emissive FromBaseColor=ScalarParameter() {group=4 Layer, parameterName=L4 Emissive From BaseColor, sliderMax=1, sortPriority=5}
   Tiling=ScalarParameter() {defaultValue=1, group=4 Layer, parameterName=L4 Tiling, sortPriority=-1}
   UV Offset=AppendVector(
    A=ScalarParameter() {group=4 Layer | UV Offsets, parameterName=L4 UV Offset X}
    B=ScalarParameter() {group=4 Layer | UV Offsets, parameterName=L4 UV Offset Y}
   )
   Texcoords=NamedRerouteUsage()
  ) {materialFunction=&MF_4Layer_Layer}
  NamedRerouteDeclaration(None=@n71.Material) {name=Layer4 Material, nodeColor=<rgba 0.9762 1 0 1>}
  NamedRerouteDeclaration(None=@n71.Mask) {name=Layer4 Mask, nodeColor=<rgba 1 1 1 1>}
  n72 = Floor(None=ComponentMask(None=TextureCoordinate() {coordinateIndex=1}) {r=true})
  n73 = Saturate(None=@n72)
  n74 = OneMinus(None=@n73)
  n75 = Saturate(None=Subtract(A=@n72))
  n76 = Subtract(A=@n73, B=@n75)
  n77 = Saturate(None=Subtract(A=@n72) {constB=2})
  n78 = Subtract(A=@n75, B=@n77)
  n79 = Constant()
  n80 = StaticSwitchParameter(
   True=Saturate(None=MaterialFunctionCall(R=@n74, G=@n76, B=@n78, A=@n77) {materialFunction=&AppendMany}.RGBA)
   False=Saturate(None=MaterialFunctionCall(R=@n74, G=@n76, B=@n75, A=@n79) {materialFunction=&AppendMany}.RGBA)
  ) {group=- Layer Switches, parameterName=Use 4 Layers, sortPriority=2}
  n81 = Saturate(None=MaterialFunctionCall(R=@n74, G=@n73, B=@n79, A=@n79) {materialFunction=&AppendMany}.RGBA)
  NamedRerouteDeclaration(None=StaticSwitchParameter(
    True=StaticSwitchParameter(True=@n80, False=@n81) {group=- Layer Switches, parameterName=Use 3 Layers, sortPriority=1}
    False=Constant4Vector() {constant=<rgba 1 1 1 1>}.RGBA
   ) {group=- Layer Switches, parameterName=Use 2 Layers, sortPriority=0}) {name=Layer Mask, nodeColor=<rgba 1 0.0168 0 1>}
  n82 = Frac(None=TextureCoordinate() {coordinateIndex=1})
  MaterialExpressionMultiply_0 = Multiply(A=OneMinus(None=@n82), B=@n82)
  n83 = Convert(None=@MaterialExpressionMultiply_0) {
   nodeName=Break Float2
   convertInputs [{
     type=Vector2
     expressionInput {
      expression=&self:MaterialExpressionMultiply_0
      inputName=None
      mask=0
      maskA=0
      maskB=0
      maskG=0
      maskR=0
      outputIndex=0
     }
    }]
   convertOutputs [{type=Scalar}, {type=Scalar}]
  }
  n84 = Multiply(A=Min(A=@n83, B=@n83)) {constB=2}
  NamedRerouteDeclaration(None=Saturate(None=OneMinus(None=@n84))) {name=Displacement Zipper Mask, nodeColor=<rgba 0 1 0.6484 1>}
  NamedRerouteDeclaration(None=@n15.Mask) {name=Layer1 Mask, nodeColor=<rgba 1 1 1 1>}
  NamedRerouteUsage()
  NamedRerouteUsage()
  NamedRerouteUsage()
  NamedRerouteUsage()
  Reroute(None=@n54.Result)
  Reroute(None=@n51.Result)
  ScalarParameter() {defaultValue=0.5, group=1 Layer, parameterName=L1 Displacement Edge Height, sliderMax=1, sortPriority=10}
  ScalarParameter() {defaultValue=0.5, group=2 Layer, parameterName=L2 Displacement Edge Height, sliderMax=1, sortPriority=10}
  ScalarParameter() {defaultValue=0.5, group=3 Layer, parameterName=L3 Displacement Edge Height, sliderMax=1, sortPriority=10}
  ScalarParameter() {defaultValue=0.5, group=4 Layer, parameterName=L4 Displacement Edge Height, sliderMax=1, sortPriority=10}
  n85 = NamedRerouteUsage()
  n86 = IfThenElse(Condition=ComponentMask(None=NamedRerouteUsage()) {g=true}, True=NamedRerouteUsage(), False=@n85)
  n87 = IfThenElse(Condition=ComponentMask(None=NamedRerouteUsage()) {b=true}, True=NamedRerouteUsage(), False=@n86)
  n88 = IfThenElse(Condition=ComponentMask(None=NamedRerouteUsage()) {a=true}, True=NamedRerouteUsage(), False=@n87)
  StaticSwitchParameter(
   True=StaticSwitchParameter(True=StaticSwitchParameter(True=@n88, False=@n87) {group=- Layer Switches, parameterName=Use 4 Layers, sortPriority=2}, False=@n86) {group=- Layer Switches, parameterName=Use 3 Layers, sortPriority=1}
   False=@n85
  ) {group=- Layer Switches, parameterName=Use 2 Layers, sortPriority=0}
 ]}
