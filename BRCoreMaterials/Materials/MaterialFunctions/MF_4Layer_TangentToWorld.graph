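# Material function that converts the normal in a Material Attributes set from tangent to world space, named for a 4-layer material setup.
# It multiplies the normal's Z by TwoSidedSign to handle back faces, transforms it, normalizes it and writes it back into the attributes.
# Layer Mask and Layer Texture Mask inputs pass straight through to matching outputs.
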MaterialFunction /BRCoreMaterials/Materials/MaterialFunctions/MF_4Layer_TangentToWorld
graph {nodes [
  n1 = GetMaterialAttributes(None=FunctionInput() {inputName=Material, inputType=FunctionInput_MaterialAttributes}) {attributeGetTypes [#1]}
  n2 = Constant2Vector() {g=1, r=1}
  n3 = TwoSidedSign()
  n4 = Transform(None=Multiply(A=@n1.Normal, B=AppendVector(A=@n2, B=@n3)) {constB=-1})
  FunctionOutput(None=SetMaterialAttributes!(MaterialAttributes=@n1.MaterialAttributes, Normal=Normalize(VectorInput=@n4)) {attributeSetTypes [#1]}) {outputName=Material}
  FunctionOutput(None=FunctionInput() {inputName=Layer Mask, inputType=FunctionInput_Vector4, sortPriority=1}) {outputName=Layer Mask, sortPriority=1}
  FunctionOutput(None=FunctionInput() {inputName=Layer Texture Mask, inputType=FunctionInput_Scalar, sortPriority=2}) {outputName=Layer Texture Mask, sortPriority=2}
 ]}
