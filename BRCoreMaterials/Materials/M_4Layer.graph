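# M_4Layer is a Default Lit master material for four-layer surfaces, built for Nanite and instanced static meshes with displacement fade enabled.
# Five material functions chain material attributes through PreFX, tangent-space effects, tangent-to-world conversion, world-space effects and PostFX, keeping all logic inside those functions.
# Tangent-space normals are disabled, so the conversion stage mid-chain lets later effects and the final normal work in world space.
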Material /BRCoreMaterials/Materials/M_4Layer
graph {
 properties {
  bAutomaticallySetUsageInEditor=false
  bEnableDisplacementFade=true
  bEnableNewHLSLGenerator=true
  bEnableNewHLSLGeneratorOverridden=true
  bTangentSpaceNormal=false
  bUseMaterialAttributes=true
  bUsedWithInstancedStaticMeshes=true
  bUsedWithNanite=true
  floatPrecisionMode=MFPM_Half_WithShaderParameters
  materialTranslatorProfile=Strict
  previewMesh=/Hera/Sets/SeaMist/Meshes/SM_SeaMist_Lighthouse_FresnelLens_A
  usedShadingModels=Default Lit
  displacementScaling {center=0, magnitude=4}
 }
 nodes [
  n1 = MaterialFunctionCall() {materialFunction=/BRCoreMaterials/Materials/MaterialFunctions/MF_4Layer_PreFX}
  n2 = MaterialFunctionCall!(Material=@n1.Material, Layer Mask=@n1.Material, Layer Texture Mask=@n1.Material) {materialFunction=/BRCoreMaterials/Materials/MaterialFunctions/MF_4Layer_Effects_TangentSpace}
  n3 = MaterialFunctionCall!(Material=@n2.Material, Layer Mask=@n2.Material, Layer Texture Mask=@n2.Material) {materialFunction=/BRCoreMaterials/Materials/MaterialFunctions/MF_4Layer_TangentToWorld}
  n4 = MaterialFunctionCall!(Material=@n3.Material, Layer Mask=@n3.Material, Layer Texture Mask=@n3.Material) {materialFunction=/BRCoreMaterials/Materials/MaterialFunctions/MF_4Layer_Effects_WorldSpace}
  MaterialFunctionCall!(Material=@n4.Material, Layer Mask=@n4.Material, Layer Texture Mask=@n4.Material) {materialFunction=/BRCoreMaterials/Materials/MaterialFunctions/MF_4Layer_PostFX}
 ]
}
